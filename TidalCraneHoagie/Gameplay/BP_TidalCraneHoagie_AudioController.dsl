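; Audio controller for the TidalCrane Hoagie vehicle that starts and stops its horn audio, which is treated as music.
; Start plays the Horn component and registers it with the MusicEventSubsystem under the Glider event tag.
; Stop fades the horn out over a second and removes that event; its Manually Stopped parameter is unused.
; Both events play a 2D music on or off cue only when the vehicle's driver is valid and locally viewed.
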(event EventStartHorn
  (bind _hoagievehicle (Variables|Default|GetHoagieVehicle))
  (bind _returnvalue (Vehicle|GetDriver _hoagievehicle))
  (Audio|Components|Audio|Play (Variables|Default|GetHorn) 0.0)
  (Audio|AddEventbyComponent (GameInstanceSubsystems|GetMusicEventSubsystem) (Variables|Default|GetHorn) "(TagName=\"MusicEventSubsystem.Event.Glider\")" "(GameplayTags=)" 0 "")
  (Utilities|IsValid _returnvalue
    (:Is%20Valid
      (if (Pawn|IsLocallyViewed _returnvalue)
        (Audio|PlaySound2D "/TidalCraneHoagie/Assets/Sounds/SC_TidalCrane_Hoagie_Music_On.SC_TidalCrane_Hoagie_Music_On")))))

(event EventStopHorn
  (params (Manually%20Stopped :type Boolean))
  (bind _hoagievehicle (Variables|Default|GetHoagieVehicle))
  (bind _returnvalue (Vehicle|GetDriver _hoagievehicle))
  (Audio|Components|Audio|FadeOut (Variables|Default|GetHorn) 1.0 0.0 "Linear" :!class /Script/Engine.AudioComponent)
  (Audio|RemoveEvent (GameInstanceSubsystems|GetMusicEventSubsystem) (Variables|Default|GetHorn))
  (Utilities|IsValid _returnvalue
    (:Is%20Valid
      (if (Pawn|IsLocallyViewed _returnvalue)
        (Audio|PlaySound2D "/TidalCraneHoagie/Assets/Sounds/SC_TidalCrane_Hoagie_Music_Off.SC_TidalCrane_Hoagie_Music_Off")))))

(fn ConstructionScript
  (|CallParentFunction))
