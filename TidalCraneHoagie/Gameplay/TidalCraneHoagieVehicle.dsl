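; TidalCraneHoagie twin-rotor helicopter vehicle, extending its parent Hoagie vehicle with model-specific FX, materials and rotor behavior.
; Enter, exit and boost events drive time-stamped jet-heat material animation, with Niagara tracking rotor speed and boosting.
; Rotor hits spare pawns between the rotors, and launch direction points away from the rotor on the pawn's side.
; It also handles playlist-based health scaling, timed vacant-vehicle crashes, remote glass-break effects and honk quest messages.
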(event EventBeginPlay
  (CallFunction|ScaleVehicleHealthbyPlaylist)
  (Variables|Vehicle|SetHoagieIdleFX_Native (Variables|Default|GetHoagie%20Idle%20FX))
  (|CallParentFunction)
  (Variables|Skydive|SetHeightToBeginSkydive (ScalableFloat|GetCachedScalableFloatValue (Variables|Default|GetRow%20ExitToSkyDiveHeight)))
  (Variables|Default|SetSwapMat_Parts_Index 1)
  (Variables|Default|SetSwapMat_Body_Index 0)
  (Variables|Default|SetVehicleMat_Parts "/TidalCraneHoagie/Assets/Materials/MI_Chopper_TwinRotor_Parts.MI_Chopper_TwinRotor_Parts")
  (Variables|Default|SetVehicleMat_Body "/TidalCraneHoagie/Assets/Materials/MI_Chopper_TwinRotor_Body.MI_Chopper_TwinRotor_Body")
  (Variables|Default|SetVehicleMat_Mask_Parts "/TidalCraneHoagie/Assets/Materials/MI_Chopper_TwinRotor_Parts_Masked.MI_Chopper_TwinRotor_Parts_Masked")
  (Variables|Default|SetVehicleMat_Mask_Body "/TidalCraneHoagie/Assets/Materials/MI_Chopper_TwinRotor_Body_Masked.MI_Chopper_TwinRotor_Body_Masked")
  (bind _returnvalue (Rendering|Material|CreateDynamicMaterialInstance (Variables|Vehicle|GetSkeletalMesh) 1 (Rendering|Material|GetMaterial (Variables|Vehicle|GetSkeletalMesh) 1) "None" :!class /Script/Engine.PrimitiveComponent))
  (Variables|Default|SetMID _returnvalue))

(event EventOnTickRotors
  (params (RotorAngleDegrees :type Float%20%28single-precision%29))
  (|CallParentFunction RotorAngleDegrees)
  (Niagara|SetNiagaraVariable(Float) (Variables|Default|GetHoagie%20Idle%20FX) "RotorSpeedAlpha" (Vehicle|GetRotorSpeedPercent)))

(event Custom|UpdateTwinRotorMaterialAnimation
  (params (Stage :type Float%20%28double-precision%29) (Animation_Duration :type Float%20%28double-precision%29) (HeatStartAmount :type Float%20%28single-precision%29) (HeatTargetAmount :type Float%20%28double-precision%29))
  (bind _mid (Variables|Default|GetMID))
  ; TwinRotor Material Animation: Jet Heat
  (Rendering|Material|SetScalarParameterValue _mid "Jet_HeatStartAmount" HeatStartAmount :!class /Script/Engine.MaterialInstanceDynamic)
  (Rendering|Material|SetScalarParameterValue _mid "Jet_HeatTargetAmount" HeatTargetAmount :!class /Script/Engine.MaterialInstanceDynamic)
  (Rendering|Material|SetScalarParameterValue _mid "Stage" Stage :!class /Script/Engine.MaterialInstanceDynamic)
  (Rendering|Material|SetScalarParameterValue _mid "Animation_Duration" Animation_Duration :!class /Script/Engine.MaterialInstanceDynamic)
  (Rendering|Material|SetScalarParameterValue _mid "GameTimeFromBP" (Utilities|Time|GetGameTimeinSeconds) :!class /Script/Engine.MaterialInstanceDynamic))

(event EventOnPawnEnterVehicle
  (params (PlayerPawn :type Fort%20Player%20Pawn%20Object%20Reference) (SeatIdx :type Integer))
  ; Vehicle Exit
  (|CallParentFunction PlayerPawn SeatIdx)
  (CallFunction|UpdateTwinRotorMaterialAnimation 1.0 2.0 0.0 0.5)
  (Utilities|Time|ClearandInvalidateTimerbyHandle (Variables|Default|GetTimer%20Handle%20No%20Passenger)))

(event EventOnPawnExitVehicle
  (params (PlayerPawn :type Fort%20Player%20Pawn%20Object%20Reference) (ExitSocketName :type Name :default "None"))
  ; Vehicle Exit
  (|CallParentFunction PlayerPawn ExitSocketName)
  (CallFunction|UpdateTwinRotorMaterialAnimation 1.0 3.0)
  (if (Vehicle|Destruction|IsCriticalHealth self)
    _
    (elif (ScalableFloat|GetValueasBool (Variables|Default|GetRow%20Vacant%20Vehicle%20Crash%20Enabled))
      (if (> (Vehicle|GetAltitude self) (ScalableFloat|GetValueatLevel (Variables|Default|GetRow%20Vacant%20Vehicle%20Crash%20Height)))
        (Utilities|Time|ClearandInvalidateTimerbyHandle (Variables|Default|GetTimer%20Handle%20No%20Passenger))
        (bind _returnvalue (Utilities|Time|SetTimerbyEvent (delegate Custom|Start%20No%20Passenger%20Crashing) 1.0))
        (Variables|Default|SetTimer%20Handle%20No%20Passenger _returnvalue)))))

(event Custom|Start%20No%20Passenger%20Crashing
  ; Vehicle Exit
  (if (and (Vehicle|HasDriver) (Vehicle|HasRiders))
    (Utilities|Time|ClearandInvalidateTimerbyHandle (Variables|Default|GetTimer%20Handle%20No%20Passenger))
    (else
      (Vehicle|StartCrashing true))))

(event EventOnBoostStarted
  ; Boost
  (|CallParentFunction)
  (Effects|Components|ParticleSystem|SetBoolParameter (Variables|Default|GetBoostActive) "IsLocal" (Pawn|IsLocallyViewed (Variables|Default|GetPilot)))
  (Effects|Components|ParticleSystem|SetBoolParameter (Variables|Default|GetHoagie%20Idle%20FX) "IsBoosting" true)
  (CallFunction|UpdateTwinRotorMaterialAnimation 1.0 1.0 0.5 1.0))

(event EventOnBoostFinished
  ; Boost
  (|CallParentFunction)
  (Effects|Components|ParticleSystem|SetBoolParameter (Variables|Default|GetHoagie%20Idle%20FX) "IsBoosting" false)
  (CallFunction|UpdateTwinRotorMaterialAnimation 2.0 3.0 0.5 1.0))

(event Vehicle|EventOnHonk
  (|CallParentFunction)
  (Default|BroadcastQuestEvent (Vehicle|GetDriver)))

(event EventOnRotorDamagePlayer
  (params (Impact :type Hit%20Result%20Structure%20%28by%20ref%29))
  (bind (_bblockinghit _binitialoverlap _time _distance _location _impactpoint) (Collision|BreakHitResult Impact))
  (bind (_returnvalue_x _returnvalue_y) (Math|Vector|BreakVector (Math|Transform|InverseTransformLocation (Transformation|GetActorTransform) _impactpoint)))
  ;[0.15 0.1293 0.02899 0.5] Don't damage pawns between the rotors
  (if (> (Math|Float|Absolute(Float) _returnvalue_y) (ScalableFloat|GetCachedScalableFloatValue (Variables|Default|GetDistanceFromCenterForRotorHit)))
    (|CallParentFunction Impact)))

(event EventOnDamageablePartHealthChanged_Event
  (params (PartConfig :type Vehicle%20Damageable%20Part%20Config%20Structure%20%28by%20ref%29) (Part :type Vehicle%20Damageable%20Part%20Structure%20%28by%20ref%29) (bNewlyDisabled :type Boolean) (bNewlyEnabled :type Boolean))
  (bind (_shapename _bonename) (Utilities|Struct|BreakVehicleDamageablePartConfig PartConfig))
  ; BREAK GLASS FX
  (if bNewlyDisabled
    (Networking|SwitchHasAuthority
      (:Remote
        (Effects|Components|ParticleSystem|SpawnFXAtLocation "/Hoagie/FX/NS_Hoagie_Window_Break_All.NS_Hoagie_Window_Break_All" (Transformation|GetSocketLocation (Variables|Vehicle|GetSkeletalMesh) (Utilities|String|StringToName (Utilities|String|Append "FX_" (Utilities|String|ToString(Name) _bonename)))) (Transformation|GetSocketRotation (Variables|Vehicle|GetSkeletalMesh) (Utilities|String|StringToName (Utilities|String|Append "FX_" (Utilities|String|ToString(Name) _bonename)))) "1.000000,1.000000,1.000000" true "AutoRelease")
        (bind _asbp_hoagie_audio_controller (Utilities|GetAudioControllerComponent)
          (:Success
            (Class|BPHoagieAudioController|OnNotifyGlassDestroyed _asbp_hoagie_audio_controller)))))))

(event Custom|ScaleVehicleHealthByPlaylist
  ;[0 0 0] Scale Vehicle Health by Playlist
  (bind (_gamestate _playlist) (Class|FortAsyncActionCurrentPlaylistReady|PlaylistDataReadyAsync)
    (:OnReady
      (bind _output_get (Variables|Default|SetMax%20Team%20Size (Variables|/Script/FortniteGame.FortPlaylist|GetMaxTeamSize _playlist)))
      (if (< _output_get 1)
        (Variables|Default|SetMax%20Team%20Size 1)
        (elif (> (Variables|Default|GetMax%20Team%20Size) 4)
          (Variables|Default|SetMax%20Team%20Size 4)
          (GameplayEffects|ApplyGameplayEffectToSelf :self (Vehicle|GetVehicleAbilitySystemComponent) :GameplayEffectClass (Variables|Default|GetGE_VehicleHealthByTeamSize) :Level (Math|Conversions|ToFloat(Integer) (- (Variables|Default|GetMax%20Team%20Size) 1))))))))

(fn BroadcastQuestEvent
  (params (Instigator :type Actor%20Object%20Reference))
  (Utilities|IsValid Instigator
    (:Is%20Valid
      (CustomVerb|BroadcastCustomVerbMessage (CustomVerb|MakeCustomVerbMessage Instigator "(TagName=\"CustomVerb.TidalCraneHoagie.HonkPressed\")")))))

(fn GetRotorLaunchDirection
  (params (HitPawn :type Actor%20Object%20Reference))
  (bind (_returnvalue_x _returnvalue_y) (Math|Vector|BreakVector (Math|Transform|InverseTransformLocation (Transformation|GetActorTransform) (Transformation|GetActorLocation HitPawn))))
  (return (- (Transformation|GetActorLocation HitPawn) (Transformation|GetSocketLocation (Variables|Vehicle|GetSkeletalMesh) (select (> _returnvalue_y 0) "FX_RotorR" "FX_RotorL")))))

(fn SpawnBooster)

(fn ConstructionScript
  (|CallParentFunction)
  (Variables|FX|SetBoost%20Active%20Template "/TidalCraneHoagie/Assets/FX/NS_TidalCraneHoagie_Boost.NS_TidalCraneHoagie_Boost"))
